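# Niagara skydiving contrail for the RippedHarvester cosmetic, with a pair of fiery meteors trailing flames, sparks and dust from the player.
# Mesh flame cores, plumes, licks, meteors and GPU sparks follow two sampled skeletal mesh bones or sockets, using partitioned particles.
# System dive velocity orients and stretches them, while an integrator value scales emitter rates, speeds and material intensity.
# Emitter-age curves and loop delays stage the meteor and flame ignition.
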NiagaraSystem /BRCosmetics/Athena/Effects/Particles/Skydiving/NS_Contrail_RippedHarvester/NS_Contrail_RippedHarvester
dependencies {
 usedDataInterfaces [
  &NiagaraDataInterfaceArrayInt32
  &NiagaraDataInterfaceMeshRendererInfo
  &NiagaraDataInterfaceColorCurve
  &NiagaraDataInterfaceCurve
  &NiagaraDataInterfaceVectorCurve
  &NiagaraDataInterfaceArrayColor
 ]
 usedDynamicInputs [
  &FN_FrontEndFloatModifier
  &SetBoolByFloatComparison
  &self:Abs_Z_Axis
  &TransformVector
 ]
 usedModules [
  &SystemState
  /Game/Effects/Niagara/Modules/Skydiving/FN_DivingAttributes/v5/FN_DivingAttributes
  &self:SystemSpawnScript.SystemScriptSource.SystemScriptGraph.NiagaraNodeAssignment_2.SetVariables_#1
  &FN_Integrator
  &self:SystemSpawnScript.SystemScriptSource.SystemScriptGraph.NiagaraNodeAssignment_1.SetVariables_#2
  &self:SystemSpawnScript.SystemScriptSource.SystemScriptGraph.NiagaraNodeAssignment_5.SetVariables_#3
  &self:SystemSpawnScript.SystemScriptSource.SystemScriptGraph.NiagaraNodeAssignment_4.SetVariables_#4
  &self:SystemSpawnScript.SystemScriptSource.SystemScriptGraph.NiagaraNodeAssignment_3.SetVariables_#5
  &self:SystemSpawnScript.SystemScriptSource.SystemScriptGraph.NiagaraNodeAssignment_0.SetVariables_#6
 ]
}
emitters {
 DustFlow {
  emitterSpawnScript [/SetVariables(Emitter.StarScale=0.75) {isSetParameters=true}]
  emitterUpdateScript [
   &EmitterState(Life Cycle Mode=::NewEnumerator1, Loop Duration=1, UseLoopDelay=false)
   &SpawnRate(SpawnRate=&Lerp_Float(A=20, B=40, Alpha=$System.FN_Integrator001.Value), Use Spawn Probability=true)
  ]
  particleSpawnScript [
   &FN_SetSpawnGroup(Number of Spawn Groups=2, Set Ribbon ID and Ribbon Link Order=true)
   /SetVariables(Particles.VisibilityTag=&SelectIntFromArray(
     Direct Array Index=&RandomRangeInt(Minimum=0, Maximum=&MakeCustomIntFromBool(Bool=$User.bIsFrontEndPreview, True INT=0, False INT=9))
     Int Selection Array=DataInterface() {propertyValues {IntData [0,0,0,0,0,0,0,0,0,1]}}
    )) {enabled=false, isSetParameters=true}
   &InitializeParticle(
    Color=&RandomRangeLinearColor(Minimum=<rgba 1.5 0.0692 0.045 1>, Maximum=<rgba 1.5 0.291 0.105 1>)
    Lifetime=&Lerp_Float(A=1.25, B=0.75, Alpha=$System.FN_Integrator001.Value)
    Mesh Scale=&MakeVector(
     X=&RandomRangeFloat(Minimum=0.5, Maximum=2)
     Y=&RandomRangeFloat(Minimum=0.5, Maximum=2)
     Z=&RandomRangeFloat(Minimum=0.5, Maximum=2)
    )
    Mesh Scale Mode=::NewEnumerator4
    Sprite Size Max=<6 6>
    Sprite Size Min=<3 3>
    Sprite Size Mode=::NewEnumerator2
    UsePositionOffset=false
    Write Material Random=true
   )
   &CylinderLocation(
    Cylinder Height=10
    Cylinder Radius=&RandomRangeFloat(Minimum=100, Maximum=300)
    MaskSpawnGroup=false
    Offset=&Multiply_VectorByFloat(
     Vector=&NormalizeVector(Vector To Normalize=$System.DiveVelocityScale)
     Float=&Lerp_Float(A=200, B=400, Alpha=$System.FN_Integrator001.Value)
    )
    Offset Coordinate Space:ENiagaraCoordinateSpace=::World
    Offset Mode:ENiagara_OffsetMode=::NewEnumerator0
    Override Seed=false
    Surface Only=true
    Use Endcaps In Surface Only Mode=false
    UseOwnerScale=true
   )
   &AddVelocity(
    Coordinate Space=::World
    Scale Added Velocity=&VectorFromFloat(Value=&Lerp_Float(A=1, B=3, Alpha=$System.FN_Integrator001.Value))
    Velocity=&Multiply_VectorByFloat(Vector=$System.DiveVelocityScale, Float=-400)
   )
   &InitialMeshOrientation(
    Rotation=&MakeVector(
     X=&RandomRangeFloat(Minimum=-1, Maximum=1)
     Y=&RandomRangeFloat(Minimum=-1, Maximum=1)
     Z=&RandomRangeFloat(Minimum=-1, Maximum=1)
    )
    Use Orientation Vector=false
    Use Rotation Vector=true
   ) {enabled=false}
  ]
  particleUpdateScript [
   &ParticleState()
   &InheritSourceMovement(ApplyForce=false, ApplyPosition=true, ApplyPositionFalloff=false)
   &SolveForcesAndVelocity(Clamp Velocity=false, Limit Acceleration=false)
   &ScaleColor(
    Scale Alpha=&FloatFromCurve(FloatCurve=DataInterface() {propertyValues {Curve {keys [<at 0 0 RCIM_Cubic>,<at 0.15 1>,<at 0.8 1 RCIM_Cubic>,<at 1 0 RCIM_Cubic>]}}})
    ScaleA=true
    ScaleRGB=true
   )
   &FN_Sprite_Attributes(
    Color=&MultiplyLinearColor(Linear Color=$Particles.Color, Scale Factor=<1 1 1 1>)
    Create Color Attribute=true
    Create SpriteSize Attribute=true
    SpriteSize=&Multiply_Vector2D(A=$Particles.Initial.SpriteSize, B=&Lerp_Vector2(A=<1 1>, B=<1 10>, Alpha=$System.FN_Integrator001.Value))
    Write to Module Namespace=true
   )
   &ScaleMeshSize(Scale Factor=&VectorFromCurve(
     VectorCurve=DataInterface() {propertyValues {
       XCurve {keys [
         {arriveTangent=2.1662, interpMode=RCIM_Cubic, leaveTangent=2.1662, tangentMode=RCTM_User, time=0, value=0.5}
         <at 0.5 1 RCIM_Cubic>
         {arriveTangent=-1.8212, interpMode=RCIM_Cubic, leaveTangent=-1.8212, tangentMode=RCTM_User, time=1, value=0.5}
        ]}
       YCurve {keys [
         {arriveTangent=2.1662, interpMode=RCIM_Cubic, leaveTangent=2.1662, tangentMode=RCTM_User, time=0, value=0.5}
         <at 0.5 1 RCIM_Cubic>
         {arriveTangent=-1.8212, interpMode=RCIM_Cubic, leaveTangent=-1.8212, tangentMode=RCTM_User, time=1, value=0.5}
        ]}
       ZCurve {keys [
         {arriveTangent=2.1662, interpMode=RCIM_Cubic, leaveTangent=2.1662, tangentMode=RCTM_User, time=0, value=0.5}
         <at 0.5 1 RCIM_Cubic>
         {arriveTangent=-1.8212, interpMode=RCIM_Cubic, leaveTangent=-1.8212, tangentMode=RCTM_User, time=1, value=0.5}
        ]}
      }}
     Scale Curve=&Lerp_Vector(A=<1 1 1>, B=<0.6 0.6 1.2>, Alpha=$System.FN_Integrator001.Value)
    )) {enabled=false}
   /SetVariables(Particles.RibbonColor:LinearColor=&MultiplyLinearColor(
     Linear Color=$Particles.Color
     Scale Factor=&MakeFloat4FromVectorAndFloat("Vector (RGB)"=<1 1 1>, "Float (Alpha)"=0.5)
    )) {enabled=false, isSetParameters=true}
  ]
  rendererClasses [&NiagaraSpriteRendererProperties, &NiagaraMeshRendererProperties]
  renderers [
   {rendererClass=&NiagaraSpriteRendererProperties}
   {rendererClass=&NiagaraSpriteRendererProperties, rendererIndex=1}
   {rendererClass=&NiagaraMeshRendererProperties, rendererIndex=2}
  ]
 }
 Flame_CenterPlume {
  emitterUpdateScript [
   &EmitterState(Delay First Loop Only=true, Life Cycle Mode=::NewEnumerator1, Loop Delay=0.5, Loop Duration=1, UseLoopDelay=true)
   &SpawnRate(SpawnRate=&Lerp_Float(A=100, B=200, Alpha=$System.FN_Integrator001.Value), Use Spawn Probability=true)
  ]
  particleSpawnScript [
   &PartitionParticles()
   &InitializeParticle(
    Color=<rgba 2 2 2 1>
    Lifetime Max=1
    Lifetime Min=0.2
    Lifetime Mode=::NewEnumerator1
    Mesh Scale=&MakeVector(
     X=&RandomRangeFloat(Minimum=0.35, Maximum=1)
     Y=&RandomRangeFloat(Minimum=0.1, Maximum=0.3)
     Z=&RandomRangeFloat(Minimum=0.1, Maximum=0.3)
    )
    Mesh Scale Mode=::NewEnumerator4
    UsePositionOffset=true
    Write Material Random=true
   )
   &InitialMeshOrientation(
    Mesh Orientation Mode=::NewEnumerator2
    Rotation=&RandomRangeVector(Minimum=<0 0 0>, Maximum=<0 0 0>)
    Rotation Coordinate Space=::NewEnumerator0
    Use Rotation Vector=false
   )
  ]
  particleUpdateScript [
   &ParticleState()
   &UpdateMeshOrientation(
    Facing Direction=$System.DiveVelocityScale
    Facing Interpolation=false
    Orientation Method=::NewEnumerator1
    Override Quat=false
    TransformFacing=false
    TransformUp=false
    Up Direction:Vector3f=$Engine.Owner.SystemYAxis
   ) {module=UpdateMeshOrientation001}
   &SkeletalMeshLocation(
    Bone / Socket Index=$Particles.PartitionParticles.Partition
    Bone Sampling Mode=::NewEnumerator4
    MaskSpawnGroup=false
    Orientation Sampling=::NewEnumerator1
    Sampled Position Offset=&Lerp_Vector(A=<10 0 0>, B=<-10 0 0>, Alpha=&MakeFloatFromInt(Int=$Particles.PartitionParticles.Partition))
    Skeletal Mesh=$User.SkeletalMesh
   )
   &ShapeLocation(
    Additional Quaternion Rotation=true
    Additional Yaw / Pitch / Roll=<0 0.25 0>
    Box / Plane Mode:ENiagara_BoxPlaneMode=::NewEnumerator1
    MaskSpawnGroup=false
    Override Seed=false
    Plane Size:Vector2f=<8 8>
    Rotation Mode=::NewEnumerator3
    Rotation Quaternion=$Output.UpdateMeshOrientation001.MeshOrientation
    Shape Origin=&ConvertPositionToVector(Input Position=$Particles.Position, Conversion Operation:ENiagara_LWCConvertPosToVec=::NewEnumerator0)
    Shape Primitive=::NewEnumerator2
    UseOwnerScale=true
   )
   /SetVariables(Particles.Position=&Add_Vector(
     A=&ConvertPositionToVector(Input Position=$Particles.Position, Conversion Operation:ENiagara_LWCConvertPosToVec=::NewEnumerator0)
     B=&Multiply_VectorByFloat(
      Vector=$System.DiveVelocityScale
      Float=&Add_Float(
       A=&FloatFromCurve(
        FloatCurve=DataInterface() {propertyValues {Curve {keys [
            {arriveTangent=0.3747, interpMode=RCIM_Cubic, leaveTangent=0.3747, tangentMode=RCTM_User, time=0, value=0}
            {arriveTangent=1.721, interpMode=RCIM_Cubic, leaveTangent=1.721, tangentMode=RCTM_User, time=1, value=1}
           ]}}}
        Scale Curve=&Multiply_Float(A=$System.DiveOffset, B=&RandomRangeFloat(Minimum=0.2, Maximum=1))
       )
       B=-11
      )
     )
    )) {isSetParameters=true, module=SetVariables_#7}
   &ScaleMeshSize(Scale Factor=&Multiply_VectorByFloat(
     Vector=&Lerp_Vector(
      A=<0.2 1 1>
      B=<3 2 2>
      Alpha=&FloatFromCurve(FloatCurve=DataInterface() {propertyValues {Curve {keys [
           {arriveTangent=0.9666, interpMode=RCIM_Cubic, time=0.2, value=0.2}
           {interpMode=RCIM_Cubic, tangentMode=RCTM_User, time=0.209, value=1}
           {interpMode=RCIM_Cubic, tangentMode=RCTM_User, time=0.3, value=1}
           {arriveTangent=-0.5118, interpMode=RCIM_Cubic, leaveTangent=-0.5118, tangentMode=RCTM_User, time=0.5, value=0.5}
           <at 1 0 RCIM_Cubic>
          ]}}})
     )
     Float=&FloatFromCurve(FloatCurve=DataInterface() {propertyValues {Curve {keys [
          {arriveTangent=3.3011, interpMode=RCIM_Cubic, leaveTangent=3.3011, tangentMode=RCTM_User, time=0, value=0.25}
          {interpMode=RCIM_Cubic, tangentMode=RCTM_User, time=0.0347, value=1}
          {arriveTangent=-0.4946, interpMode=RCIM_Cubic, leaveTangent=-0.4946, tangentMode=RCTM_User, time=0.3, value=0.85}
          {arriveTangent=-0.1763, interpMode=RCIM_Cubic, leaveTangent=-0.1763, tangentMode=RCTM_User, time=1, value=0.25}
         ]}}})
    ))
   &DynamicMaterialParameters(
    Index 0 Param 1=&FloatFromCurve(
     FloatCurve=DataInterface() {propertyValues {Curve {keys [
         <at 0 0 RCIM_Cubic>
         {arriveTangent=1, interpMode=RCIM_Cubic, leaveTangent=1, time=0.6, value=0.375}
         {arriveTangent=1.6037, interpMode=RCIM_Cubic, leaveTangent=1.6037, tangentMode=RCTM_User, time=1, value=1}
        ]}}}
     Scale Curve=$System.Emissive
    )
    Module.Param0WriteEnabled=true
    Module.Param1WriteEnabled=false
    Module.Param2WriteEnabled=false
    Module.Param3WriteEnabled=false
   ) {enabled=false}
   &Color(Color=&LerpLinearColors(
     StartColor=$User.ColorA
     EndColor=$User.ColorB
     LerpFactor=&FloatFromCurve(FloatCurve=DataInterface() {propertyValues {Curve {keys [{interpMode=RCIM_Cubic, tangentMode=RCTM_User, time=0.4, value=0}, {interpMode=RCIM_Cubic, leaveTangent=5.1459, time=0.75, value=1}]}}})
    ))
   &DynamicMaterialParameters(
    Index 0 Param 1=&FloatFromCurve(
     FloatCurve=DataInterface() {propertyValues {Curve {keys [
         {arriveTangent=0.0957, interpMode=RCIM_Cubic, leaveTangent=0.0957, tangentMode=RCTM_User, time=0, value=1}
         {
          arriveTangent=-0.7712
          arriveTangentWeight=0.0795
          interpMode=RCIM_Cubic
          leaveTangent=-0.7712
          leaveTangentWeight=0.1591
          tangentMode=RCTM_User
          time=0.1
          value=0.55
         }
         {arriveTangent=-0.0687, interpMode=RCIM_Cubic, leaveTangent=-0.0687, tangentMode=RCTM_User, time=0.5, value=0.5}
         {arriveTangentWeight=0.1364, interpMode=RCIM_Cubic, leaveTangentWeight=0.1023, tangentMode=RCTM_User, time=0.6414, value=0.0255}
         {arriveTangent=-0.2361, interpMode=RCIM_Cubic, leaveTangent=-0.2361, tangentMode=RCTM_User, time=1, value=0}
        ]}}}
     Scale Curve=0.75
    )
    Module.Param0WriteEnabled=true
    Module.Param1WriteEnabled=false
    Module.Param2WriteEnabled=false
    Module.Param3WriteEnabled=false
   ) {module=DynamicMaterialParameters001}
  ]
  rendererClasses [&NiagaraMeshRendererProperties]
  renderers [{rendererClass=&NiagaraMeshRendererProperties, rendererIndex=0}]
 }
 Flame_Core {
  emitterUpdateScript [
   &EmitterState(
    Inactive Response=::NewEnumerator1
    Life Cycle Mode=::NewEnumerator1
    Loop Behavior=::NewEnumerator1
    Loop Duration Mode=::NewEnumerator1
    UseLoopDelay=false
   ) {module=EmitterState001}
   &SpawnBurst_Instantaneous(Spawn Count=2, Use Spawn Probability=false, UseLoopCountLimit=false)
   /SetVariables(Emitter.WPO_Speed=1.2) {isSetParameters=true, module=SetVariables_#8}
  ]
  particleSpawnScript [
   &PartitionParticles()
   &InitializeParticle(
    Lifetime=2
    Mesh Scale Mode=::NewEnumerator3
    Sprite Size Mode=::NewEnumerator3
    Uniform Sprite Size=1
    UsePositionOffset=false
    Write Material Random=true
   )
   &InitialMeshOrientation(
    Mesh Axis To Orient=<0 0 0>
    Mesh Orientation Mode=::NewEnumerator2
    Orientation Vector=<0 -1 0>
    Rotation Coordinate Space=::NewEnumerator0
    Use Rotation Vector=false
   )
   /SetVariables(Particles.LightRadius=25) {isSetParameters=true, module=SetVariables_#9}
  ]
  particleUpdateScript [
   &ParticleState(
    Kill Particles When Lifetime Has Elapsed=false
    Let Infinitely Lived Particles Die When Emitter Deactivates=true
    Loop Particles Lifetime=true
   )
   &SkeletalMeshLocation(
    Bone / Socket Index=$Particles.PartitionParticles.Partition
    Bone Sampling Mode=::NewEnumerator4
    MaskSpawnGroup=false
    Orientation Sampling=::NewEnumerator1
    Sampled Position Offset=&Lerp_Vector(A=<15 0 0>, B=<-15 0 0>, Alpha=&MakeFloatFromInt(Int=$Particles.PartitionParticles.Partition))
    Skeletal Mesh=$User.SkeletalMesh
   )
   &UpdateMeshOrientation(
    Facing Direction=$System.DiveVelocityScale
    Facing Interpolation=false
    Orientation Method=::NewEnumerator1
    Override Quat=false
    Side Direction:Vector3f=$Engine.Owner.SystemXAxis
    Stabilization Mode:ENiagara_OrientVectorCount=::NewEnumerator2
    TransformFacing=false
    TransformSide=false
   ) {module=UpdateMeshOrientation001}
   &ScaleMeshSize(Scale Factor=&VectorFromFloat(Value=&FloatFromCurve(
      FloatCurve=DataInterface() {propertyValues {Curve {keys [
          {interpMode=RCIM_Cubic, tangentMode=RCTM_User, time=0.4986, value=0}
          {arriveTangent=0.0224, interpMode=RCIM_Cubic, leaveTangent=0.0224, tangentMode=RCTM_User, time=0.7939, value=1}
         ]}}}
      CurveIndex=$Emitter.Age
     ))) {module=ScaleMeshSize001}
   &ScaleMeshSize(Scale Factor=&Multiply_Vector(
     A=<1 1 1>
     B=&VectorFromCurve(
      VectorCurve=DataInterface() {propertyValues {
        XCurve {keys [
          {arriveTangent=-6.6682, interpMode=RCIM_Cubic, leaveTangent=-6.6682, leaveTangentWeight=0.1989, tangentMode=RCTM_User, time=0, value=1}
          {arriveTangentWeight=0.0333, interpMode=RCIM_Cubic, leaveTangentWeight=0.0333, tangentMode=RCTM_User, time=0.1, value=0.4}
          {
           arriveTangent=8.3022
           arriveTangentWeight=0.0333
           interpMode=RCIM_Cubic
           leaveTangent=-8.3169
           leaveTangentWeight=0.0333
           tangentMode=RCTM_Break
           time=0.2
           value=1
          }
          {arriveTangentWeight=0.0333, interpMode=RCIM_Cubic, leaveTangentWeight=0.0333, tangentMode=RCTM_User, time=0.3, value=0.3}
          {arriveTangent=8.5, arriveTangentWeight=0.2233, interpMode=RCIM_Cubic, leaveTangent=8.5, leaveTangentWeight=0.1989, time=0.4, value=1}
          {arriveTangentWeight=0.0333, interpMode=RCIM_Cubic, leaveTangentWeight=0.0333, tangentMode=RCTM_User, time=0.5, value=2}
          {arriveTangent=-12.82, arriveTangentWeight=0.0333, interpMode=RCIM_Cubic, leaveTangent=-12.82, tangentMode=RCTM_User, time=0.6, value=1}
         ]}
        YCurve {keys [
          {arriveTangent=-6.6682, interpMode=RCIM_Cubic, leaveTangent=-6.6682, tangentMode=RCTM_User, time=0, value=1}
          <at 0.1 0.4 RCIM_Cubic>
          {arriveTangent=7, interpMode=RCIM_Cubic, leaveTangent=7, time=0.2, value=1}
          <at 0.3 1.8 RCIM_Cubic>
          {arriveTangent=-8, interpMode=RCIM_Cubic, leaveTangent=-8, time=0.4, value=1}
          <at 0.5 0.2 RCIM_Cubic>
          {arriveTangent=11.695, interpMode=RCIM_Cubic, leaveTangent=11.695, tangentMode=RCTM_User, time=0.6, value=1}
         ]}
        ZCurve {keys [
          {arriveTangent=5.8832, interpMode=RCIM_Cubic, leaveTangent=5.8832, leaveTangentWeight=0.2248, tangentMode=RCTM_User, time=0, value=1}
          {arriveTangentWeight=0.0333, interpMode=RCIM_Cubic, leaveTangentWeight=0.0333, tangentMode=RCTM_User, time=0.1, value=1.6}
          {arriveTangent=-6.5, arriveTangentWeight=0.2081, interpMode=RCIM_Cubic, leaveTangent=-6.5, leaveTangentWeight=0.1872, time=0.2, value=1}
          {arriveTangentWeight=0.0333, interpMode=RCIM_Cubic, leaveTangentWeight=0.0333, tangentMode=RCTM_User, time=0.3, value=0.3}
          {
           arriveTangent=6.625
           arriveTangentWeight=0.0333
           interpMode=RCIM_Cubic
           leaveTangent=-5.8832
           leaveTangentWeight=0.0333
           tangentMode=RCTM_Break
           time=0.4
           value=1
          }
          {arriveTangentWeight=0.0333, interpMode=RCIM_Cubic, leaveTangentWeight=0.0333, tangentMode=RCTM_User, time=0.5, value=0.2}
          {arriveTangent=11.695, arriveTangentWeight=0.0333, interpMode=RCIM_Cubic, leaveTangent=11.695, tangentMode=RCTM_User, time=0.6, value=1}
         ]}
       }}
      CurveIndex=$Emitter.Age
      Scale Curve=&VectorFromFloat(Value=&FloatFromCurve(
        FloatCurve=DataInterface() {propertyValues {Curve {keys [
            {arriveTangent=12.2362, interpMode=RCIM_Cubic, leaveTangent=12.2362, tangentMode=RCTM_User, time=0, value=0}
            {interpMode=RCIM_Cubic, tangentMode=RCTM_User, time=0.2, value=1.4}
            {arriveTangent=-0.25, interpMode=RCIM_Cubic, leaveTangent=-0.25, time=0.4, value=0.7}
            {arriveTangent=0.8571, interpMode=RCIM_Cubic, leaveTangent=0.8571, time=0.6, value=1.3}
            <at 0.75 1 RCIM_Cubic>
           ]}}}
        CurveIndex=$Emitter.Age
       ))
     )
    )) {enabled=false}
   /SetVariables(Particles.RandomTime:NiagaraFloat=&Frac_Float(Float=&Add_Float(
      A=$System.NormalizedLoopAge
      B=&Multiply_FloatByInt(Float=0.25, Integer=$Particles.PartitionParticles.Partition)
     ))) {isSetParameters=true, module=SetVariables_#10}
   &DynamicMaterialParameters(
    Index 0 Param 1=&Lerp_Float(
     A=0
     B=0.4
     Alpha=&FloatFromCurve(
      FloatCurve=DataInterface() {propertyValues {Curve {keys [
          <at 0 0 RCIM_Cubic>
          {arriveTangent=4.1653, interpMode=RCIM_Cubic, leaveTangent=4.1653, time=0.0611, value=0.8}
          {arriveTangent=1.291, interpMode=RCIM_Cubic, leaveTangent=1.291, time=0.1801, value=0.75}
          {arriveTangent=-2.9431, interpMode=RCIM_Cubic, leaveTangent=-2.9431, time=0.216, value=1}
          {arriveTangent=-2.6528, interpMode=RCIM_Cubic, leaveTangent=-2.6528, time=0.316, value=0.35}
          {arriveTangent=-2.6149, interpMode=RCIM_Cubic, leaveTangent=-2.6149, time=0.3516, value=0.6401}
          {interpMode=RCIM_Cubic, tangentMode=RCTM_User, time=0.4498, value=0}
          {arriveTangent=3.486, interpMode=RCIM_Cubic, leaveTangent=3.486, time=0.5, value=0.8}
          {arriveTangent=-0.8651, interpMode=RCIM_Cubic, leaveTangent=-0.8651, time=0.5932, value=0.5}
          {arriveTangent=-4.2832, interpMode=RCIM_Cubic, leaveTangent=-4.2832, time=0.6325, value=0.6853}
          {arriveTangent=-1.4195, interpMode=RCIM_Cubic, leaveTangent=-1.4195, time=0.71, value=0}
          {arriveTangent=2.1663, interpMode=RCIM_Cubic, leaveTangent=2.1663, time=0.7983, value=0.45}
          {arriveTangent=5.4095, interpMode=RCIM_Cubic, leaveTangent=5.4095, time=0.8715, value=0.35}
          {arriveTangent=-2.7248, interpMode=RCIM_Cubic, leaveTangent=-2.7248, time=0.9, value=1}
          <at 1 0 RCIM_Cubic>
         ]}}}
      CurveIndex=$Particles.RandomTime
      Scale Curve=&Lerp_Float(A=0.4, B=1, Alpha=$System.FN_Integrator001.Value)
     )
    )
    Module.Param0WriteEnabled=true
    Module.Param1WriteEnabled=false
    Module.Param2WriteEnabled=false
    Module.Param3WriteEnabled=false
   ) {enabled=false}
   /SetVariables(Particles.DynamicMaterialParameter1:Vector4f=&MakeFloat4FromVectorAndFloat(
     "Vector (RGB)"=$System.DiveVelocityScale
     "Float (Alpha)"=&FloatFromCurve(
      FloatCurve=DataInterface() {propertyValues {Curve {keys [<at 0.5 0 RCIM_Cubic>, {interpMode=RCIM_Cubic, tangentMode=RCTM_User, time=0.7, value=2}, <at 0.9 1 RCIM_Cubic>]}}}
      CurveIndex=$Emitter.Age
      Scale Curve=&Lerp_Float(A=8, B=24, Alpha=$System.FN_Integrator001.Value)
     )
    )) {isSetParameters=true}
   &DynamicMaterialParameters(
    Index 0 Param 1=1
    Module.Param0WriteEnabled=true
    Module.Param1WriteEnabled=false
    Module.Param2WriteEnabled=false
    Module.Param3WriteEnabled=false
   ) {enabled=false, module=DynamicMaterialParameters001}
   &Color(Color=&LerpLinearColors(
     StartColor=$User.ColorA
     EndColor=$User.ColorB
     LerpFactor=&Lerp_Float(
      A=&FloatFromCurve(
       FloatCurve=DataInterface() {propertyValues {Curve {keys [<at 0 0 RCIM_Cubic>, <at 0.1 1 RCIM_Cubic>, {interpMode=RCIM_Cubic, tangentMode=RCTM_User, time=0.6, value=0}]}}}
       CurveIndex=$Emitter.Age
      )
      B=0.33
      Alpha=&FloatFromCurve(
       FloatCurve=DataInterface() {propertyValues {Curve {keys [
           <at 0 0 RCIM_Cubic>
           {arriveTangent=3.8876, interpMode=RCIM_Cubic, leaveTangent=3.8876, time=0.0611, value=0.75}
           {arriveTangent=1.6137, interpMode=RCIM_Cubic, leaveTangent=1.6137, time=0.1801, value=0.7}
           {arriveTangent=-2.5752, interpMode=RCIM_Cubic, leaveTangent=-2.5752, time=0.216, value=1}
           {arriveTangent=-2.6528, interpMode=RCIM_Cubic, leaveTangent=-2.6528, time=0.316, value=0.35}
           {arriveTangent=-2.6149, interpMode=RCIM_Cubic, leaveTangent=-2.6149, time=0.3516, value=0.6401}
           {interpMode=RCIM_Cubic, tangentMode=RCTM_User, time=0.4498, value=0}
           {arriveTangent=3.486, interpMode=RCIM_Cubic, leaveTangent=3.486, time=0.5, value=1}
           {arriveTangent=-2.3742, interpMode=RCIM_Cubic, leaveTangent=-2.3742, time=0.5932, value=0.5}
           {arriveTangent=-4.2832, interpMode=RCIM_Cubic, leaveTangent=-4.2832, time=0.6325, value=0.6853}
           {arriveTangent=-0.2132, interpMode=RCIM_Cubic, leaveTangent=-0.2132, time=0.71, value=0}
           {arriveTangent=2.1663, interpMode=RCIM_Cubic, leaveTangent=2.1663, time=0.7983, value=0.65}
           {arriveTangent=3.4424, interpMode=RCIM_Cubic, leaveTangent=3.4424, time=0.8715, value=0.35}
           {arriveTangent=-2.7248, interpMode=RCIM_Cubic, leaveTangent=-2.7248, time=0.9, value=1}
           <at 1 0 RCIM_Cubic>
          ]}}}
       CurveIndex=$Particles.RandomTime
       Scale Curve=&Lerp_Float(A=0.4, B=1, Alpha=$System.FN_Integrator001.Value)
      )
     )
    )) {enabled=false}
   &FN_Sprite_Attributes(
    Camera Offset=20
    Color=&MultiplyLinearColor(
     Linear Color=&LerpLinearColors(
      StartColor=$User.ColorA
      EndColor=$User.ColorB
      LerpFactor=&Lerp_Float(
       A=0.9
       B=1
       Alpha=&FloatFromCurve(
        FloatCurve=DataInterface() {propertyValues {Curve {keys [
            <at 0 0 RCIM_Cubic>
            {arriveTangent=1.304, interpMode=RCIM_Cubic, leaveTangent=1.304, time=0.1, value=0.75}
            {arriveTangent=-1.2483, interpMode=RCIM_Cubic, leaveTangent=-1.2483, time=0.1681, value=0.2192}
            {arriveTangent=-0.9034, interpMode=RCIM_Cubic, leaveTangent=-0.9034, time=0.216, value=1}
            {arriveTangent=-0.1111, interpMode=RCIM_Cubic, leaveTangent=-0.1111, time=0.3, value=0.1}
            {arriveTangent=1.8356, interpMode=RCIM_Cubic, leaveTangent=1.8356, time=0.3516, value=0.6401}
            {arriveTangent=2.4257, interpMode=RCIM_Cubic, leaveTangent=2.4257, time=0.4219, value=0.05}
            {arriveTangent=0.2031, interpMode=RCIM_Cubic, leaveTangent=0.2031, time=0.5, value=1}
            {arriveTangent=-2.3742, interpMode=RCIM_Cubic, leaveTangent=-2.3742, time=0.5703, value=0.3538}
            {arriveTangent=-1.957, interpMode=RCIM_Cubic, leaveTangent=-1.957, time=0.6325, value=0.6853}
            {arriveTangent=-0.2171, interpMode=RCIM_Cubic, leaveTangent=-0.2171, time=0.7, value=0}
            {arriveTangent=1.2231, interpMode=RCIM_Cubic, leaveTangent=1.2231, time=0.7714, value=0.6552}
            {arriveTangent=1.1257, interpMode=RCIM_Cubic, leaveTangent=1.1257, time=0.8336, value=0.2634}
            {arriveTangent=-1.5833, interpMode=RCIM_Cubic, leaveTangent=-1.5833, time=0.9, value=1}
            <at 1 0 RCIM_Cubic>
           ]}}}
        CurveIndex=$System.NormalizedLoopAge
        Scale Curve=&Lerp_Float(A=1, B=2, Alpha=$System.FN_Integrator001.Value)
       )
      )
     )
     Scale Factor=&MakeFloat4FromVectorAndFloat(
      "Vector (RGB)"=<1 1 1>
      "Float (Alpha)"=&FloatFromCurve(
       FloatCurve=DataInterface() {propertyValues {Curve {keys [<at 0.5 0 RCIM_Cubic>, {interpMode=RCIM_Cubic, tangentMode=RCTM_User, time=1, value=1}]}}}
       CurveIndex=$Emitter.Age
      )
     )
    )
    Create Camera Offset Attribute=true
    Create Color Attribute=true
    Write to Module Namespace=true
   ) {enabled=false}
   &SolveForcesAndVelocity(Clamp Velocity=false, Limit Acceleration=false)
  ]
  rendererClasses [&NiagaraMeshRendererProperties]
  renderers [{rendererClass=&NiagaraMeshRendererProperties, rendererIndex=0}]
 }
 Flame_MeshLicks {
  emitterUpdateScript [
   &EmitterState(Delay First Loop Only=true, Life Cycle Mode=::NewEnumerator1, Loop Delay=0.5, Loop Duration=1, UseLoopDelay=true)
   &SpawnRate(SpawnRate=&Lerp_Float(A=200, B=400, Alpha=$System.FN_Integrator001.Value), Use Spawn Probability=true)
  ]
  particleSpawnScript [
   &InitializeParticle(
    Color=<rgba 2 2 2 1>
    Lifetime Max=0.8
    Lifetime Min=0.2
    Lifetime Mode=::NewEnumerator1
    Mesh Scale=&MakeVector(
     X=&RandomRangeFloat(Minimum=0.3, Maximum=1)
     Y=&RandomRangeFloat(Minimum=0.1, Maximum=0.3)
     Z=&RandomRangeFloat(Minimum=0.1, Maximum=0.3)
    )
    Mesh Scale Mode=::NewEnumerator4
    UsePositionOffset=true
    Write Material Random=true
   )
   &PartitionParticles()
   &InitialMeshOrientation(
    Mesh Orientation Mode=::NewEnumerator2
    Rotation=&RandomRangeVector(Minimum=<50 0 0>, Maximum=<-50 0 0>)
    Rotation Coordinate Space=::NewEnumerator0
    Use Rotation Vector=true
   )
  ]
  particleUpdateScript [
   &ParticleState()
   &UpdateMeshOrientation(
    Facing Direction=$System.DiveVelocityScale
    Facing Interpolation=false
    Orientation Method=::NewEnumerator1
    Override Quat=false
    TransformFacing=false
    TransformUp=false
    Up Direction:Vector3f=$Engine.Owner.SystemYAxis
   ) {module=UpdateMeshOrientation001}
   &SkeletalMeshLocation(
    Bone / Socket Index=$Particles.PartitionParticles.Partition
    Bone Sampling Mode=::NewEnumerator4
    MaskSpawnGroup=false
    Orientation Sampling=::NewEnumerator1
    Sampled Position Offset=&Lerp_Vector(A=<10 0 0>, B=<-10 0 0>, Alpha=&MakeFloatFromInt(Int=$Particles.PartitionParticles.Partition))
    Skeletal Mesh=$User.SkeletalMesh
   )
   &ShapeLocation(
    Additional Quaternion Rotation=true
    Additional Yaw / Pitch / Roll=<0 0.25 0>
    Box / Plane Mode:ENiagara_BoxPlaneMode=::NewEnumerator1
    MaskSpawnGroup=false
    Override Seed=false
    Plane Size:Vector2f=<16 16>
    Rotation Mode=::NewEnumerator3
    Rotation Quaternion=$Output.UpdateMeshOrientation001.MeshOrientation
    Shape Origin=&ConvertPositionToVector(Input Position=$Particles.Position, Conversion Operation:ENiagara_LWCConvertPosToVec=::NewEnumerator0)
    Shape Primitive=::NewEnumerator2
    UseOwnerScale=true
   )
   /SetVariables(Particles.Position=&Add_Vector(
     A=&ConvertPositionToVector(Input Position=$Particles.Position, Conversion Operation:ENiagara_LWCConvertPosToVec=::NewEnumerator0)
     B=&Multiply_VectorByFloat(
      Vector=$System.DiveVelocityScale
      Float=&Add_Float(
       A=&FloatFromCurve(
        FloatCurve=DataInterface() {propertyValues {Curve {keys [
            {arriveTangent=0.3747, interpMode=RCIM_Cubic, leaveTangent=0.3747, tangentMode=RCTM_User, time=0, value=0}
            {arriveTangent=1.721, interpMode=RCIM_Cubic, leaveTangent=1.721, tangentMode=RCTM_User, time=1, value=1}
           ]}}}
        Scale Curve=&Multiply_Float(A=$System.DiveOffset, B=&RandomRangeFloat(Minimum=0.2, Maximum=0.8))
       )
       B=-11
      )
     )
    )) {isSetParameters=true, module=SetVariables_#7}
   &ScaleMeshSize(Scale Factor=&Multiply_VectorByFloat(
     Vector=&Lerp_Vector(
      A=<1 1 1>
      B=&RandomRangeVector(Minimum=<0.5 1.5 6>, Maximum=<3 3 3>)
      Alpha=&FloatFromCurve(FloatCurve=DataInterface() {propertyValues {Curve {keys [
           <at 0.2 0 RCIM_Cubic>
           {interpMode=RCIM_Cubic, tangentMode=RCTM_User, time=0.2785, value=1}
           {interpMode=RCIM_Cubic, tangentMode=RCTM_User, time=0.4, value=1}
           {arriveTangent=-2.6857, interpMode=RCIM_Cubic, leaveTangent=-2.6857, tangentMode=RCTM_User, time=0.9, value=0.1}
          ]}}})
     )
     Float=&FloatFromCurve(FloatCurve=DataInterface() {propertyValues {Curve {keys [
          {arriveTangent=3.3011, interpMode=RCIM_Cubic, leaveTangent=3.3011, tangentMode=RCTM_User, time=0, value=0.25}
          {interpMode=RCIM_Cubic, tangentMode=RCTM_User, time=0.0347, value=1}
          {arriveTangent=-0.1446, interpMode=RCIM_Cubic, leaveTangent=-0.1446, tangentMode=RCTM_User, time=0.4669, value=0.9}
          {arriveTangent=-2.7832, interpMode=RCIM_Cubic, leaveTangent=-2.7832, tangentMode=RCTM_User, time=1, value=0.25}
         ]}}})
    ))
   &DynamicMaterialParameters(
    Index 0 Param 1=&FloatFromCurve(
     FloatCurve=DataInterface() {propertyValues {Curve {keys [
         <at 0.07 0 RCIM_Cubic>
         {arriveTangent=0.0957, interpMode=RCIM_Cubic, leaveTangent=0.0957, tangentMode=RCTM_User, time=0.15, value=1}
         {arriveTangent=-0.6992, interpMode=RCIM_Cubic, leaveTangent=-0.6992, tangentMode=RCTM_User, time=0.5, value=0.2}
         {arriveTangent=-0.2361, interpMode=RCIM_Cubic, leaveTangent=-0.2361, tangentMode=RCTM_User, time=1, value=0}
        ]}}}
     Scale Curve=4
    )
    Module.Param0WriteEnabled=true
    Module.Param1WriteEnabled=false
    Module.Param2WriteEnabled=false
    Module.Param3WriteEnabled=false
   ) {module=DynamicMaterialParameters001}
   &SolveForcesAndVelocity(Clamp Velocity=false, Limit Acceleration=false)
  ]
  rendererClasses [&NiagaraMeshRendererProperties]
  renderers [{rendererClass=&NiagaraMeshRendererProperties, rendererIndex=0}]
 }
 GPU_Sparks {
  simTarget=GPUComputeSim
  emitterUpdateScript [
   &EmitterState(Delay First Loop Only=true, Life Cycle Mode=::NewEnumerator1, Loop Delay=0.4803, Loop Duration=2.0196, UseLoopDelay=true)
   &SpawnRate(SpawnRate=3600, Use Spawn Probability=false)
  ]
  particleSpawnScript [
   &InitializeParticle(
    Color=&SelectLinearColorFromArray(
     Array Sampling Mode=::NewEnumerator0
     Direct Array Index=?
     Color Selection Array=DataInterface() {propertyValues {ColorData [<rgba 4 0.0133 0 1>, <rgba 0.401 0.0134 0 1>]}}
     Random Seed=0
    )
    Lifetime Max=0.75
    Lifetime Min=0.35
    Lifetime Mode=::NewEnumerator1
    Sprite Size Mode=::NewEnumerator1
    Uniform Sprite Size Max=5
    Uniform Sprite Size Min=3
    UsePositionOffset=false
    Write Material Random=true
   )
   &PartitionParticles()
   &FN_SetSpawnGroup(Number of Spawn Groups=2)
   &AddVelocity(
    Coordinate Space=::World
    Scale Added Velocity=&VectorFromFloat(Value=&Lerp_Float(A=1, B=3, Alpha=$System.FN_Integrator001.Value))
    Velocity=&Multiply_VectorByFloat(Vector=$System.DiveVelocityScale, Float=&RandomRangeFloat(Minimum=-50, Maximum=-300))
   ) {module=AddVelocity001}
   &DynamicMaterialParameters(
    Index 0 Param 1=HlslExpression() {hlslExpression="rand(1)"}
    Module.Param0WriteEnabled=true
    Module.Param1WriteEnabled=false
    Module.Param2WriteEnabled=false
    Module.Param3WriteEnabled=false
   ) {enabled=false}
  ]
  particleUpdateScript [
   &ParticleState()
   &UpdateMeshOrientation(
    Facing Direction=$System.DiveVelocityScale
    Facing Interpolation=false
    Orientation Method=::NewEnumerator1
    Override Quat=false
    TransformFacing=false
    TransformUp=false
    Up Direction:Vector3f=$Engine.Owner.SystemYAxis
   ) {module=UpdateMeshOrientation001}
   &SkeletalMeshLocation(
    Bone / Socket Index=$Particles.PartitionParticles.Partition
    Bone Sampling Mode=::NewEnumerator4
    MaskSpawnGroup=false
    Orientation Sampling=::NewEnumerator1
    Sampled Position Offset=&Lerp_Vector(A=<15 0 0>, B=<-15 0 0>, Alpha=&MakeFloatFromInt(Int=$Particles.PartitionParticles.Partition))
    Skeletal Mesh=$User.SkeletalMesh
   ) {module=SkeletalMeshLocation003}
   &ShapeLocation(
    MaskSpawnGroup=false
    Offset=<0 0 10>
    Offset Mode=::NewEnumerator0
    Override Seed=false
    Ring Radius=12
    Shape Origin=$StackContext.Position
    Shape Primitive=::NewEnumerator4
    UseOwnerScale=true
   ) {module=ShapeLocation001}
   /Game/Effects/Niagara/Modules/Skydiving/FN_DivingAttributes/FN_DivingAttributes_v2(Accumulate Dive=1, Accumulate Rest=0, ParaGlideLeanAlpha:NiagaraFloat=$User.ParaGlideLeanAlpha)
   /SetVariables(Particles.Position=&Add_Vector(
     A=&ConvertPositionToVector(Input Position=$Particles.Position, Convert To Absolute Large World Vector=false)
     B=&Add_Vector(A=&Multiply_VectorByFloat(Vector=$System.DiveVelocityScale, Float=10), B=<0 0 0>)
    )) {enabled=false, isSetParameters=true}
   &InheritVelocity(Inherited Velocity Amount Scale=<0.5 0.3 -0.05>, Inherited Velocity Speed Limit=0, Speed Limit=true, UseSpeedThreshold=false) {enabled=false}
   &CurlNoiseForce(
    Noise Frequency=200
    Noise Quality / Cost=::NewEnumerator0
    Noise Strength=&FloatFromCurve(
     FloatCurve=DataInterface() {propertyValues {Curve {keys [{interpMode=RCIM_Cubic, tangentMode=RCTM_User, time=0.3, value=0.2}, {interpMode=RCIM_Cubic, tangentMode=RCTM_User, time=0.6, value=1}]}}}
     Scale Curve=2000
    )
    Pan Noise=true
    Pan Noise Field=<15 10 10>
   ) {enabled=false}
   &CurlNoiseForce(Noise Frequency=10000, Noise Strength=1000, Pan Noise=false, Pan Noise Field=<-10 -7 0>) {enabled=false, module=CurlNoiseForce001}
   &Drag(Drag=&RandomRangeFloat(Minimum=1, Maximum=3), Use Linear Drag=true, Use Rotational Drag=true) {enabled=false}
   &SolveForcesAndVelocity(Clamp Velocity=false, Limit Acceleration=false)
   &AlignSpriteToMeshOrientation(SetAlignment=true, SetFacing=true) {enabled=false}
   &ScaleSpriteSize(
    Uniform Curve Scale=&RandomRangeFloat(Minimum=0.5, Maximum=2)
    Uniform Curve Sprite Scale=DataInterface() {propertyValues {Curve {keys [
        {arriveTangent=9.0725, interpMode=RCIM_Cubic, leaveTangent=9.0725, tangentMode=RCTM_User, time=0, value=0}
        {arriveTangent=-2.5249, interpMode=RCIM_Cubic, leaveTangent=-2.5249, tangentMode=RCTM_User, time=0.0562, value=2}
        {arriveTangent=-2.1191, interpMode=RCIM_Cubic, leaveTangent=-2.1191, time=0.4964, value=0.75}
        {arriveTangent=-5.2602, interpMode=RCIM_Cubic, leaveTangent=-5.2602, tangentMode=RCTM_User, time=1, value=0}
       ]}}}
   )
   &ScaleSpriteSizeBySpeed(
    Max Scale Factor=<0.3 4>
    Min Scale Factor=<0.3 0.3>
    Velocity Threshold:NiagaraFloat=&RandomRangeFloat(Minimum=200, Maximum=1000)
   )
   &Color(
    Color=&RandomRangeLinearColor(Minimum=<rgba 10 0.002 0 1>, Maximum=<rgba 1 1 0 0.5>)
    Scale Alpha=&FloatFromCurve(FloatCurve=DataInterface() {propertyValues {Curve {keys [
         {tangentMode=RCTM_User, time=0, value=1}
         {arriveTangent=-0.0405, interpMode=RCIM_Cubic, leaveTangent=-0.0405, tangentMode=RCTM_User, time=0.7009, value=1}
         {arriveTangent=-8.0089, interpMode=RCIM_Cubic, leaveTangent=-8.0089, tangentMode=RCTM_User, time=1, value=0}
        ]}}})
   ) {enabled=false}
  ]
  rendererClasses [&NiagaraSpriteRendererProperties]
  renderers [{rendererClass=&NiagaraSpriteRendererProperties}]
 }
 Meteor_Mesh {
  emitterUpdateScript [
   &EmitterState(
    Inactive Response=::NewEnumerator1
    Life Cycle Mode=::NewEnumerator1
    Loop Behavior=::NewEnumerator1
    Loop Duration Mode=::NewEnumerator1
    UseLoopDelay=false
   ) {module=EmitterState001}
   &SpawnBurst_Instantaneous(Spawn Count=2, Spawn Time=-0.3583, Use Spawn Probability=false, UseLoopCountLimit=false)
   /SetVariables(Emitter.WPO_Speed=1.2) {isSetParameters=true, module=SetVariables_#8}
   &SpawnBurst_Instantaneous(Spawn Count=2, Use Spawn Probability=false, UseLoopCountLimit=false) {module=SpawnBurst_Instantaneous001}
  ]
  particleSpawnScript [
   &PartitionParticles()
   &InitializeParticle(
    Color=<rgba 200 0.4 10.3798 1>
    Lifetime=2
    Mesh Scale Mode=::NewEnumerator3
    Sprite Size Mode=::NewEnumerator3
    Uniform Sprite Size=1
    UsePositionOffset=false
    Write Material Random=true
   )
   &InitialMeshOrientation(
    Mesh Axis To Orient=<0 0 0>
    Mesh Orientation Mode=::NewEnumerator2
    Orientation Vector=<0 -1 0>
    Rotation Coordinate Space=::NewEnumerator0
    Use Rotation Vector=false
   )
   /SetVariables(Particles.LightRadius=25) {isSetParameters=true, module=SetVariables_#9}
  ]
  particleUpdateScript [
   &ParticleState(
    Kill Particles When Lifetime Has Elapsed=false
    Let Infinitely Lived Particles Die When Emitter Deactivates=true
    Loop Particles Lifetime=true
   )
   &SkeletalMeshLocation(
    Bone / Socket Index=$Particles.PartitionParticles.Partition
    Bone Sampling Mode=::NewEnumerator4
    MaskSpawnGroup=false
    Orientation Sampling=::NewEnumerator1
    Sampled Position Offset=&Lerp_Vector(A=<15 0 0>, B=<-15 0 0>, Alpha=&MakeFloatFromInt(Int=$Particles.PartitionParticles.Partition))
    Skeletal Mesh=$User.SkeletalMesh
   )
   &UpdateMeshOrientation(
    Facing Direction=$System.DiveVelocityScale
    Facing Interpolation=false
    Orientation Method=::NewEnumerator1
    Override Quat=false
    Side Direction:Vector3f=$Engine.Owner.SystemXAxis
    Stabilization Mode:ENiagara_OrientVectorCount=::NewEnumerator2
    TransformFacing=false
    TransformSide=false
   ) {module=UpdateMeshOrientation001}
   /SetVariables(Particles.RandomTime:NiagaraFloat=&Frac_Float(Float=&Add_Float(
      A=$System.NormalizedLoopAge
      B=&Multiply_FloatByInt(Float=0.25, Integer=$Particles.PartitionParticles.Partition)
     ))) {isSetParameters=true, module=SetVariables_#10}
   &ScaleMeshSize(Scale Factor=&VectorFromFloat(Value=&FloatFromCurve(
      FloatCurve=DataInterface() {propertyValues {Curve {keys [
          {arriveTangent=1.2682, interpMode=RCIM_Cubic, leaveTangent=1.2682, tangentMode=RCTM_User, time=0, value=0}
          {arriveTangent=0.0224, interpMode=RCIM_Cubic, leaveTangent=0.0224, tangentMode=RCTM_User, time=0.5, value=1}
         ]}}}
      CurveIndex=$Emitter.Age
     )))
   &DynamicMaterialParameters(
    Index 3 Param 1=&Lerp_Float(
     A=2
     B=8
     Alpha=&FloatFromCurve(
      FloatCurve=DataInterface() {propertyValues {Curve {keys [{arriveTangent=0.0815, interpMode=RCIM_Cubic, leaveTangent=0.0815, tangentMode=RCTM_User, time=1, value=0}, <at 1.05 1 RCIM_Cubic>]}}}
      CurveIndex=$Emitter.Age
     )
    )
    Index 3 Param 2=&Lerp_Float(
     A=777
     B=777
     Alpha=&FloatFromCurve(
      FloatCurve=DataInterface() {propertyValues {Curve {keys [
          {arriveTangent=4.1785, interpMode=RCIM_Cubic, leaveTangent=4.1785, tangentMode=RCTM_User, time=0, value=0}
          {arriveTangent=1.1179, interpMode=RCIM_Cubic, leaveTangent=1.1179, tangentMode=RCTM_User, time=0.2042, value=0.6228}
          {arriveTangent=0.0224, interpMode=RCIM_Cubic, leaveTangent=0.0224, tangentMode=RCTM_User, time=1, value=1}
         ]}}}
      CurveIndex=$Emitter.Age
     )
    )
    Index 3 Param 3=&Lerp_Float(
     A=100
     B=2.5
     Alpha=&FloatFromCurve(
      FloatCurve=DataInterface() {propertyValues {Curve {keys [
          {arriveTangent=0.6645, interpMode=RCIM_Cubic, leaveTangent=0.6645, leaveTangentWeight=1.4322, tangentMode=RCTM_User, time=0, value=0.3}
          {arriveTangent=0.0014, interpMode=RCIM_Cubic, leaveTangent=0.0014, tangentMode=RCTM_User, time=1.2, value=1}
          {arriveTangent=0.0224, interpMode=RCIM_Cubic, leaveTangent=0.0224, tangentMode=RCTM_User, time=1.5, value=1}
         ]}}}
      CurveIndex=$Emitter.Age
     )
    )
    Index 3 Param 4=&Lerp_Float(
     A=1.2
     B=1.2
     Alpha=&FloatFromCurve(
      FloatCurve=DataInterface() {propertyValues {Curve {keys [
          {arriveTangent=4.1785, interpMode=RCIM_Cubic, leaveTangent=4.1785, tangentMode=RCTM_User, time=0, value=0}
          {arriveTangent=1.1179, interpMode=RCIM_Cubic, leaveTangent=1.1179, tangentMode=RCTM_User, time=0.2042, value=0.6228}
          {arriveTangent=0.0224, interpMode=RCIM_Cubic, leaveTangent=0.0224, tangentMode=RCTM_User, time=1, value=1}
         ]}}}
      CurveIndex=$Emitter.Age
     )
    )
    Module.Param0WriteEnabled3=true
    Module.Param1WriteEnabled3=true
    Module.Param2WriteEnabled3=true
    Module.Param3WriteEnabled3=true
    Write Parameter Index 0=false
    Write Parameter Index 3=true
   )
   &SolveForcesAndVelocity(Clamp Velocity=false, Limit Acceleration=false)
  ]
  rendererClasses [&NiagaraLightRendererProperties, &NiagaraMeshRendererProperties]
  renderers [
   {rendererClass=&NiagaraLightRendererProperties, rendererIndex=0}
   {rendererClass=&NiagaraMeshRendererProperties, rendererIndex=1}
  ]
 }
 RockShards_Floaters {
  emitterUpdateScript [
   &EmitterState(Delay First Loop Only=true, Life Cycle Mode=::NewEnumerator1, Loop Delay=0.2416, Loop Duration=1.2584, UseLoopDelay=true)
   &SpawnRate(SpawnRate=&Lerp_Float(A=12, B=40, Alpha=$System.FN_Integrator001.Value), Use Spawn Probability=true)
  ]
  particleSpawnScript [
   &PartitionParticles()
   &InitializeParticle(
    Lifetime Mode=::NewEnumerator1
    Mesh Scale=&MakeVector(
     X=&RandomRangeFloat(Minimum=0.75, Maximum=1)
     Y=&RandomRangeFloat(Minimum=0.75, Maximum=1)
     Z=&RandomRangeFloat(Minimum=0.75, Maximum=1)
    )
    Mesh Scale Mode=::NewEnumerator4
    UsePositionOffset=true
    Write Material Random=true
   )
   &ShapeLocation(
    Apply To Particle Position=::NewEnumerator1
    MaskSpawnGroup=false
    Override Seed=false
    Shape Origin=&VectorFromFloat(Value=0)
    Sphere Radius=16
    UseOwnerScale=true
   )
  ]
  particleUpdateScript [
   &ParticleState()
   &UpdateMeshOrientation(Rotation Rate=2)
   &UpdateMeshOrientation(
    Facing Direction=$System.DiveVelocityScale
    Facing Interpolation=false
    Orientation Method=::NewEnumerator1
    Override Quat=false
    TransformFacing=false
    TransformUp=false
    Up Direction:Vector3f=$Engine.Owner.SystemYAxis
   ) {enabled=false, module=UpdateMeshOrientation001}
   &SkeletalMeshLocation(
    Bone / Socket Index=$Particles.PartitionParticles.Partition
    Bone Sampling Mode=::NewEnumerator4
    MaskSpawnGroup=false
    Orientation Sampling=::NewEnumerator1
    Sampled Position Offset=$StackContext.ShapeLocation.ShapePosition
    Skeletal Mesh=$User.SkeletalMesh
   )
   /SetVariables(Particles.Position=&Add_Vector(
     A=Linked() {linkedVariable {name=Particles.Position, type=NiagaraPosition}}
     B=&Add_Vector(
      A=&Multiply_VectorByFloat(
       Vector=$System.DiveVelocityScale
       Float=&Add_Float(
        A=&FloatFromCurve(
         FloatCurve=DataInterface() {propertyValues {Curve {keys [
             {arriveTangent=0.5, interpMode=RCIM_Cubic, leaveTangent=0.5, tangentMode=RCTM_User, time=0, value=0.1}
             {arriveTangent=1.5776, interpMode=RCIM_Cubic, leaveTangent=1.5776, tangentMode=RCTM_User, time=1, value=1}
            ]}}}
         Scale Curve=&Multiply_Float(A=$System.DiveOffset, B=&RandomRangeFloat(Minimum=0.2, Maximum=1))
        )
        B=0
       )
      )
      B=&MakeVector(
       X=&FloatFromCurve(
        FloatCurve=DataInterface() {propertyValues {Curve {keys [
            {arriveTangent=9.0725, interpMode=RCIM_Cubic, leaveTangent=9.0725, tangentMode=RCTM_User, time=0, value=0}
            {arriveTangent=0.357, interpMode=RCIM_Cubic, leaveTangent=0.357, tangentMode=RCTM_User, time=0.3, value=0.8}
            {arriveTangent=-0.8571, interpMode=RCIM_Cubic, leaveTangent=-0.8571, time=0.6, value=0.5}
            {arriveTangent=0.0679, interpMode=RCIM_Cubic, leaveTangent=0.0679, tangentMode=RCTM_User, time=1, value=0.2}
           ]}}}
        Scale Curve=&Multiply_Float(A=&RandomRangeFloat(Minimum=-1, Maximum=1), B=24)
       )
       Y=&FloatFromCurve(
        FloatCurve=DataInterface() {propertyValues {Curve {keys [
            {arriveTangent=9.0725, interpMode=RCIM_Cubic, leaveTangent=9.0725, tangentMode=RCTM_User, time=0, value=0}
            {arriveTangentWeight=0.1161, interpMode=RCIM_Cubic, leaveTangentWeight=0.1335, tangentMode=RCTM_User, time=0.3, value=1.05}
            {arriveTangent=-1.2143, interpMode=RCIM_Cubic, leaveTangent=-1.2143, time=0.6, value=0.7}
            {arriveTangent=-0.612, interpMode=RCIM_Cubic, leaveTangent=-0.612, tangentMode=RCTM_User, time=1, value=0.2}
           ]}}}
        Scale Curve=&Multiply_Float(A=&RandomRangeFloat(Minimum=-1, Maximum=1), B=24)
       )
       Z=0
      )
     )
    )) {isSetParameters=true}
   &ScaleMeshSize(Scale Factor=&Multiply_VectorByFloat(
     Vector=&Lerp_Vector(
      A=<0.5 1 1>
      B=<1 0.75 1>
      Alpha=&FloatFromCurve(
       FloatCurve=DataInterface() {propertyValues {Curve {keys [
           {arriveTangent=0.9666, interpMode=RCIM_Cubic, time=0.2, value=0}
           {interpMode=RCIM_Cubic, tangentMode=RCTM_User, time=0.3, value=1}
           {interpMode=RCIM_Cubic, tangentMode=RCTM_User, time=0.6, value=1}
           <at 0.9 0>
          ]}}}
       Scale Curve=1.75
      )
     )
     Float=&FloatFromCurve(
      FloatCurve=DataInterface() {propertyValues {Curve {keys [
          {arriveTangent=3.3011, interpMode=RCIM_Cubic, leaveTangent=3.3011, tangentMode=RCTM_User, time=0, value=0.65}
          {interpMode=RCIM_Cubic, tangentMode=RCTM_User, time=0.1, value=1}
          {interpMode=RCIM_Cubic, tangentMode=RCTM_Break, time=0.5, value=1}
          {arriveTangent=-1.8681, interpMode=RCIM_Cubic, leaveTangent=-1.8681, tangentMode=RCTM_User, time=1, value=0.25}
         ]}}}
      Scale Curve=0.35
     )
    ))
   &MeshRotationForce(Rotation=&RandomRangeVector(Minimum=<-20 -20 -20>, Maximum=<20 20 20>)) {enabled=false}
   &ScaleVelocity(Velocity Scale=&VectorFromCurve(
     VectorCurve=DataInterface() {propertyValues {
       XCurve {keys [
         {interpMode=RCIM_Cubic, tangentMode=RCTM_User, time=0, value=0}
         {arriveTangent=2.6641, interpMode=RCIM_Cubic, leaveTangent=2.6641, tangentMode=RCTM_User, time=1, value=1}
        ]}
       YCurve {keys [
         {interpMode=RCIM_Cubic, tangentMode=RCTM_User, time=0, value=0}
         {arriveTangent=2.6641, interpMode=RCIM_Cubic, leaveTangent=2.6641, tangentMode=RCTM_User, time=1, value=1}
        ]}
       ZCurve {keys [
         {interpMode=RCIM_Cubic, tangentMode=RCTM_User, time=0, value=0}
         {arriveTangent=2.6641, interpMode=RCIM_Cubic, leaveTangent=2.6641, tangentMode=RCTM_User, time=1, value=1}
        ]}
      }}
     Scale Curve=<1 1 100>
    )) {enabled=false}
   &SolveForcesAndVelocity(Clamp Velocity=false, Limit Acceleration=false)
  ]
  rendererClasses [&NiagaraMeshRendererProperties]
  renderers [{rendererClass=&NiagaraMeshRendererProperties, rendererIndex=0}]
 }
 RockShards_Tail {
  emitterUpdateScript [
   &EmitterState(Delay First Loop Only=true, Life Cycle Mode=::NewEnumerator1, Loop Delay=0.2416, Loop Duration=1.2584, UseLoopDelay=true)
   &SpawnRate(SpawnRate=&Lerp_Float(A=40, B=100, Alpha=$System.FN_Integrator001.Value), Use Spawn Probability=true)
  ]
  particleSpawnScript [
   &PartitionParticles()
   &InitializeParticle(
    Color=<rgba 0.3281 0.3281 0.3281 1>
    Lifetime Max=1.2
    Lifetime Min=0.4
    Lifetime Mode=::NewEnumerator1
    Mesh Scale=&MakeVector(
     X=&RandomRangeFloat(Minimum=0.35, Maximum=0.75)
     Y=&RandomRangeFloat(Minimum=0.35, Maximum=0.75)
     Z=&RandomRangeFloat(Minimum=0.5, Maximum=1)
    )
    Mesh Scale Mode=::NewEnumerator4
    UsePositionOffset=true
    Write Material Random=true
   )
   &ShapeLocation(
    Apply To Particle Position=::NewEnumerator1
    MaskSpawnGroup=false
    Offset=&MaskVectorBySpawnGroup(Masked Vector=<10 0 0>, Passthrough Vector=<-10 0 0>, Spawn Group Mask=$Particles.PartitionParticles.Partition)
    Offset Mode=::NewEnumerator0
    Override Seed=false
    Shape Origin=&VectorFromFloat(Value=0)
    Sphere Radius=20
    UseOwnerScale=true
   )
   &AddVelocity(Velocity=<0 0 100>) {enabled=false}
  ]
  particleUpdateScript [
   &ParticleState()
   &SkeletalMeshLocation(
    Bone / Socket Index=$Particles.PartitionParticles.Partition
    Bone Sampling Mode=::NewEnumerator4
    MaskSpawnGroup=false
    Orientation Sampling=::NewEnumerator1
    Sampled Position Offset=$StackContext.ShapeLocation.ShapePosition
    Skeletal Mesh=$User.SkeletalMesh
   )
   /SetVariables(Particles.Position=&Add_Vector(
     A=Linked() {linkedVariable {name=Particles.Position, type=NiagaraPosition}}
     B=&Multiply_VectorByFloat(
      Vector=$System.DiveVelocityScale
      Float=&Add_Float(
       A=&FloatFromCurve(
        FloatCurve=DataInterface() {propertyValues {Curve {keys [
            {arriveTangent=0.3747, interpMode=RCIM_Cubic, leaveTangent=0.3747, tangentMode=RCTM_User, time=0, value=0}
            {arriveTangent=1.721, interpMode=RCIM_Cubic, leaveTangent=1.721, tangentMode=RCTM_User, time=1, value=1}
           ]}}}
        Scale Curve=&Multiply_Float(A=$System.DiveOffset, B=&RandomRangeFloat(Minimum=0.5, Maximum=1))
       )
       B=-11
      )
     )
    )) {isSetParameters=true, module=SetVariables_#7}
   &ScaleMeshSize(Scale Factor=&Multiply_VectorByFloat(
     Vector=&Lerp_Vector(
      A=<0.8 0.8 0.8>
      B=<1.2 1.2 1.2>
      Alpha=&FloatFromCurve(
       FloatCurve=DataInterface() {propertyValues {Curve {keys [
           {arriveTangent=0.9666, interpMode=RCIM_Cubic, time=0.0046, value=0}
           {interpMode=RCIM_Cubic, tangentMode=RCTM_User, time=0.0288, value=1}
           {arriveTangent=0.038, interpMode=RCIM_Cubic, leaveTangent=0.038, tangentMode=RCTM_User, time=0.4, value=0.95}
           {arriveTangent=-2.0607, interpMode=RCIM_Cubic, leaveTangent=-2.0607, tangentMode=RCTM_User, time=0.9, value=0}
          ]}}}
       Scale Curve=0.6
      )
     )
     Float=&FloatFromCurve(FloatCurve=DataInterface() {propertyValues {Curve {keys [
          {arriveTangent=20.3126, interpMode=RCIM_Cubic, leaveTangent=20.3126, tangentMode=RCTM_User, time=0, value=0.25}
          {interpMode=RCIM_Cubic, tangentMode=RCTM_User, time=0.0541, value=1.2}
          {arriveTangent=-0.7209, interpMode=RCIM_Cubic, leaveTangent=-0.7209, time=0.2338, value=0.85}
          {interpMode=RCIM_Cubic, leaveTangent=-1.1083, tangentMode=RCTM_Break, time=0.5397, value=0.85}
          {arriveTangent=-0.3747, interpMode=RCIM_Cubic, leaveTangent=-0.3747, tangentMode=RCTM_User, time=1, value=0.25}
         ]}}})
    ))
   &MeshRotationForce(Rotation=&RandomRangeVector(Minimum=<-20 -20 -20>, Maximum=<20 20 20>))
   &SolveForcesAndVelocity(Clamp Velocity=false, Limit Acceleration=false)
   &AlignSpriteToMeshOrientation(SetAlignment=true, SetFacing=true) {enabled=false}
  ]
  rendererClasses [&NiagaraMeshRendererProperties]
  renderers [{rendererClass=&NiagaraMeshRendererProperties, rendererIndex=0}]
 }
 Smoke_Tail {
  emitterUpdateScript [
   &EmitterState(Life Cycle Mode=::NewEnumerator1, Loop Behavior=::NewEnumerator1, Loop Delay=0.5, Loop Duration Mode=::NewEnumerator1, UseLoopDelay=true)
   &SpawnBurst_Instantaneous(Spawn Count=2, Use Spawn Probability=false, UseLoopCountLimit=false)
  ]
  particleSpawnScript [
   &PartitionParticles()
   &InitializeParticle(
    Lifetime=2
    Mesh Scale=&FN_FrontEndVector3Modifier(Gameplay=<1 1 1>, Front End=<1 1 0.8>)
    Mesh Scale Mode=::NewEnumerator4
    Sprite Size Mode=::NewEnumerator3
    Uniform Sprite Size=50
    UsePositionOffset=false
    Write Material Random=true
   )
   &InitialMeshOrientation(
    Mesh Axis To Orient=<0 0 -1>
    Mesh Orientation Mode=::NewEnumerator2
    Rotation=&Lerp_Vector(A=<0 0 0>, B=<0.5 0 0>, Alpha=&MakeFloatFromInt(Int=$Particles.PartitionParticles.Partition))
    Rotation Coordinate Space=::NewEnumerator0
    Use Rotation Vector=true
   )
  ]
  particleUpdateScript [
   &ParticleState(
    Kill Particles When Lifetime Has Elapsed=false
    Let Infinitely Lived Particles Die When Emitter Deactivates=true
    Loop Particles Lifetime=true
   )
   &SkeletalMeshLocation(
    Bone / Socket Index=$Particles.PartitionParticles.Partition
    Bone Sampling Mode=::NewEnumerator4
    MaskSpawnGroup=false
    Orientation Sampling=::NewEnumerator1
    Sampled Position Offset=&Lerp_Vector(A=<10 0 0>, B=<-10 0 0>, Alpha=&MakeFloatFromInt(Int=$Particles.PartitionParticles.Partition))
    Skeletal Mesh=$User.SkeletalMesh
   )
   /SetVariables(Particles.Position=&Add_Vector(
     A=&ConvertPositionToVector(Input Position=$Particles.Position, Convert To Absolute Large World Vector=false)
     B=&Add_Vector(A=&Multiply_VectorByFloat(Vector=$System.DiveVelocityScale, Float=10), B=<0 0 0>)
    )) {isSetParameters=true}
   &UpdateMeshOrientation(
    Facing Direction=$System.DiveVelocityScale
    Facing Interpolation=false
    Orientation Method=::NewEnumerator1
    Override Quat=false
    Side Direction:Vector3f=$Engine.Owner.SystemXAxis
    Stabilization Mode:ENiagara_OrientVectorCount=::NewEnumerator2
    TransformFacing=false
    TransformSide=false
   ) {module=UpdateMeshOrientation001}
   &UpdateMeshOrientation(Rotation Rate=-1, Rotation Vector=<0 0 1>)
   &CameraOffset(Camera Offset Amount=8)
   &ScaleMeshSize(Scale Factor=&Multiply_Vector(
     A=&Lerp_Vector(A=<1.2 1.2 2>, B=<1 1 1.5>, Alpha=$System.FN_Integrator001.Value)
     B=&MakeVector(
      X=1
      Y=1
      Z=&FloatFromCurve(
       FloatCurve=DataInterface() {propertyValues {Curve {keys [<at 0 0 RCIM_Cubic>, {arriveTangent=4.1531, interpMode=RCIM_Cubic, leaveTangent=4.1531, tangentMode=RCTM_User, time=0.5, value=1}]}}}
       CurveIndex=$Emitter.Age
      )
     )
    ))
   &ScaleColor(
    Scale Alpha=&FloatFromCurve(
     FloatCurve=DataInterface() {propertyValues {Curve {keys [<at 0.5 0 RCIM_Cubic>, {arriveTangent=4.1531, interpMode=RCIM_Cubic, leaveTangent=4.1531, tangentMode=RCTM_User, time=1, value=1}]}}}
     CurveIndex=$Emitter.Age
     Scale Curve=0.5
    )
    Scale RGB=&Multiply_VectorByFloat(Vector=&MakeVectorFromLinearColorRGB(Color=<rgba 0.45 0.4 0.5 1>), Float=1)
    ScaleA=true
    ScaleRGB=true
   ) {module=ScaleColor001}
   /SetVariables(Particles.Color_Yellow:LinearColor=&MakeLinearColorFromVectorAndFloat(
     "Vector (RGB)"=&Multiply_VectorByFloat(Vector=&MakeVectorFromLinearColorRGB(Color=$User.ColorB), Float=5)
     "Float (Alpha)"=&FloatFromCurve(
      FloatCurve=DataInterface() {propertyValues {Curve {keys [<at 0 0 RCIM_Cubic>, {arriveTangent=2.7731, interpMode=RCIM_Cubic, leaveTangent=2.7731, tangentMode=RCTM_User, time=1, value=1}]}}}
      CurveIndex=$Emitter.Age
     )
    )) {isSetParameters=true}
   /SetVariables(Particles.Scale_Yellow:Vector3f=&Multiply_Vector(A=$Particles.Scale, B=<0.75 0.75 0.5>)) {isSetParameters=true}
   &ScaleColor(
    Scale Alpha=&FloatFromCurve(
     FloatCurve=DataInterface() {propertyValues {Curve {keys [{interpMode=RCIM_Cubic, tangentMode=RCTM_User, time=0, value=1}, {interpMode=RCIM_Cubic, tangentMode=RCTM_User, time=1, value=1}]}}}
     CurveIndex=$Emitter.Age
    )
    Scale RGB=&MakeVectorFromLinearColorRGB(Color=&LerpLinearColors(
      StartColor=$User.ColorA
      EndColor=$User.ColorB
      LerpFactor=&Lerp_Float(
       A=0
       B=3
       Alpha=&FloatFromCurve(
        FloatCurve=DataInterface() {propertyValues {Curve {keys [
            <at 0 0 RCIM_Cubic>
            {arriveTangent=1.304, interpMode=RCIM_Cubic, leaveTangent=1.304, time=0.1, value=0.75}
            {arriveTangent=-1.2483, interpMode=RCIM_Cubic, leaveTangent=-1.2483, time=0.1681, value=0.2192}
            {arriveTangent=-0.9034, interpMode=RCIM_Cubic, leaveTangent=-0.9034, time=0.216, value=1}
            {arriveTangent=-0.1111, interpMode=RCIM_Cubic, leaveTangent=-0.1111, time=0.3, value=0.1}
            {arriveTangent=1.8356, interpMode=RCIM_Cubic, leaveTangent=1.8356, time=0.3516, value=0.6401}
            {arriveTangent=2.4257, interpMode=RCIM_Cubic, leaveTangent=2.4257, time=0.4219, value=0.05}
            {arriveTangent=0.2031, interpMode=RCIM_Cubic, leaveTangent=0.2031, time=0.5, value=1}
            {arriveTangent=-2.3742, interpMode=RCIM_Cubic, leaveTangent=-2.3742, time=0.5703, value=0.3538}
            {arriveTangent=-1.957, interpMode=RCIM_Cubic, leaveTangent=-1.957, time=0.6325, value=0.6853}
            {arriveTangent=-0.2171, interpMode=RCIM_Cubic, leaveTangent=-0.2171, time=0.7, value=0}
            {arriveTangent=1.2231, interpMode=RCIM_Cubic, leaveTangent=1.2231, time=0.7714, value=0.6552}
            {arriveTangent=1.1257, interpMode=RCIM_Cubic, leaveTangent=1.1257, time=0.8336, value=0.2634}
            {arriveTangent=-1.5833, interpMode=RCIM_Cubic, leaveTangent=-1.5833, time=0.9, value=1}
            <at 1 0 RCIM_Cubic>
           ]}}}
        CurveIndex=$System.NormalizedLoopAge
        Scale Curve=&Lerp_Float(A=0.1, B=1, Alpha=$System.FN_Integrator001.Value)
       )
      )
     ))
    ScaleA=true
    ScaleRGB=true
   )
   &InterpolateOverTime(
    Rate Of Change=10
    Target Vector Value:Vector3f=&ConvertPositionToVector(Input Position=$Particles.Position, Convert To Absolute Large World Vector=false)
    Variable Type=::NewEnumerator2
   ) {enabled=false}
   /SetVariables(Particles.InterpolateDynamic:Vector4f=&MakeFloat4FromVectorAndFloat("Vector (RGB)"=$Output.InterpolateOverTime.Moving Average, "Float (Alpha)"=1)) {enabled=false, isSetParameters=true}
   &DynamicMaterialParameters(
    Index 0 Param 2=&RandomRangeFloat(Minimum=0, Maximum=1.5)
    Index 2 Param 3=0.1
    Index 2 Param 4=1.75
    Module.Param0WriteEnabled=true
    Module.Param0WriteEnabled2=false
    Module.Param1WriteEnabled=true
    Module.Param1WriteEnabled2=false
    Module.Param2WriteEnabled=false
    Module.Param2WriteEnabled2=true
    Module.Param3WriteEnabled=true
    Module.Param3WriteEnabled2=true
    Write Parameter Index 2=true
   )
  ]
  rendererClasses [&NiagaraMeshRendererProperties]
  renderers [
   {rendererClass=&NiagaraMeshRendererProperties, rendererIndex=0}
   {rendererClass=&NiagaraMeshRendererProperties, rendererIndex=1}
  ]
 }
 Sparks_EnterBurst001 {
  simTarget=GPUComputeSim
  emitterUpdateScript [
   &EmitterState(Life Cycle Mode=::NewEnumerator1, Loop Behavior=::NewEnumerator1, Loop Delay=0.55, Loop Duration=3.5532, UseLoopDelay=true)
   &SpawnRate(
    SpawnRate=&FloatFromCurve(
     FloatCurve=DataInterface() {propertyValues {Curve {keys [
         <at 0.55 0 RCIM_Cubic>
         {arriveTangent=1.5046, interpMode=RCIM_Cubic, leaveTangent=1.5046, tangentMode=RCTM_User, time=0.7, value=0.1}
         {arriveTangentWeight=0.0898, interpMode=RCIM_Cubic, leaveTangentWeight=0.2693, tangentMode=RCTM_User, time=0.9, value=1}
         <at 2 0 RCIM_Cubic>
        ]}}}
     CurveIndex=$Emitter.Age
     Scale Curve=2000
    )
    Use Spawn Probability=false
   )
  ]
  particleSpawnScript [
   &InitializeParticle(
    Color=&SelectLinearColorFromArray(
     Array Sampling Mode=::NewEnumerator0
     Direct Array Index=?
     Color Selection Array=DataInterface() {propertyValues {ColorData [<rgba 10 0.0333 0 1>, <rgba 3 0.05 0 1>]}}
     Random Seed=0
    )
    Lifetime Max=0.6
    Lifetime Min=0.1
    Lifetime Mode=::NewEnumerator1
    Sprite Size Mode=::NewEnumerator1
    Uniform Sprite Size Max=2
    Uniform Sprite Size Min=0.5
    UsePositionOffset=false
    Write Material Random=true
   )
   &PartitionParticles()
   &SkeletalMeshLocation(
    Bone / Socket Index=$Particles.PartitionParticles.Partition
    Bone Sampling Mode=::NewEnumerator4
    MaskSpawnGroup=false
    Position Sampling=::NewEnumerator2
    Sampled Position Offset=&MaskVectorBySpawnGroup(Masked Vector=<16 0 0>, Passthrough Vector=<-16 0 0>, Spawn Group Mask=$Particles.PartitionParticles.Partition)
    Skeletal Mesh=$User.SkeletalMesh
   ) {module=SkeletalMeshLocation001}
   &ShapeLocation(
    MaskSpawnGroup=false
    Override Seed=false
    Shape Origin=$Particles.SkeletalMeshLocation001.SampledPosition
    Sphere Radius=&FloatFromCurve(
     FloatCurve=DataInterface() {propertyValues {Curve {keys [
         <at 0 0.2 RCIM_Cubic>
         {arriveTangent=0.8, interpMode=RCIM_Cubic, leaveTangent=0.8, time=0.8, value=1}
         {arriveTangent=-0.0149, interpMode=RCIM_Cubic, leaveTangent=-0.0149, tangentMode=RCTM_User, time=1, value=1}
         <at 1.5 0.35 RCIM_Cubic>
        ]}}}
     CurveIndex=$Emitter.Age
     Scale Curve=20
    )
    UseOwnerScale=true
   ) {module=ShapeLocation001}
   &AddVelocity(
    Velocity=&Lerp_Vector(
     A=&RandomRangeVector(Minimum=<0 0 200>, Maximum=<0 0 300>)
     B=&FN_FrontEndVector3Modifier(Gameplay=&RandomRangeVector(Minimum=<0 0 100>, Maximum=<0 0 200>), Front End=<0 0 220>)
     Alpha=$System.FN_Integrator001.Value
    )
    Velocity Speed Scale=&FloatFromCurve(
     FloatCurve=DataInterface() {propertyValues {Curve {keys [
         <at 0 0 RCIM_Cubic>
         {interpMode=RCIM_Cubic, tangentMode=RCTM_User, time=0.5, value=0}
         {arriveTangentWeight=0.1222, interpMode=RCIM_Cubic, leaveTangentWeight=0.4886, tangentMode=RCTM_User, time=0.7, value=1}
         {arriveTangentWeight=0.1333, interpMode=RCIM_Cubic, tangentMode=RCTM_User, time=1, value=1}
        ]}}}
     CurveIndex=$Emitter.Age
    )
   )
  ]
  particleUpdateScript [
   &ParticleState()
   &FN_DivingAttributes_v2(Accumulate Dive=1, Accumulate Rest=0, ParaGlideLeanAlpha:NiagaraFloat=$User.ParaGlideLeanAlpha)
   &CurlNoiseForce(
    Noise Frequency=&RandomRangeFloat(Minimum=60, Maximum=400, Evaluation Type=::NewEnumerator1)
    Noise Strength=&Multiply_Float(
     A=2000
     B=&FloatFromCurve(
      FloatCurve=DataInterface() {propertyValues {Curve {keys [
          {interpMode=RCIM_Cubic, tangentMode=RCTM_User, time=0.5, value=1}
          {arriveTangentWeight=0.1465, interpMode=RCIM_Cubic, leaveTangentWeight=0.3419, tangentMode=RCTM_User, time=0.6954, value=1}
          {interpMode=RCIM_Cubic, tangentMode=RCTM_User, time=0.7954, value=0}
         ]}}}
      CurveIndex=$Emitter.Age
     )
    )
    Pan Noise=false
    Pan Noise Field=<100 3 3>
   )
   &CurlNoiseForce(
    Noise Frequency=600
    Noise Strength=2000
    Pan Noise=true
    Pan Noise Field=&Multiply_VectorByFloat(
     Vector=<1 1 1>
     Float=&FloatFromCurve(FloatCurve=DataInterface() {propertyValues {Curve {keys [
          {interpMode=RCIM_Cubic, leaveTangentWeight=0.4297, tangentMode=RCTM_User, time=0, value=0.1}
          {arriveTangent=0.0224, interpMode=RCIM_Cubic, leaveTangent=0.0224, tangentMode=RCTM_User, time=0.7, value=0.1}
          {arriveTangentWeight=0.1744, interpMode=RCIM_Cubic, leaveTangentWeight=0.2325, tangentMode=RCTM_User, time=1, value=1}
          <at 2 1 RCIM_Cubic>
         ]}}})
    )
   ) {module=CurlNoiseForce001}
   &Drag(Drag=&RandomRangeFloat(Minimum=8, Maximum=12), Use Linear Drag=true, Use Rotational Drag=true)
   &SolveForcesAndVelocity(Clamp Velocity=false, Limit Acceleration=false)
   &ScaleSpriteSize(
    Scale Factor=&Lerp_Vector2(
     A=<1 1>
     B=<1.5 1.5>
     Alpha=&FloatFromCurve(
      FloatCurve=DataInterface() {propertyValues {Curve {keys [
          {arriveTangent=9.0725, interpMode=RCIM_Cubic, leaveTangent=9.0725, tangentMode=RCTM_User, time=0.5, value=0}
          {arriveTangent=-2.5249, interpMode=RCIM_Cubic, leaveTangent=-2.5249, tangentMode=RCTM_User, time=0.6, value=3.75}
          {arriveTangent=-5.6855, interpMode=RCIM_Cubic, leaveTangent=-5.6855, tangentMode=RCTM_User, time=0.8, value=0.6}
          {arriveTangent=-1.4144, interpMode=RCIM_Cubic, leaveTangent=-1.4144, tangentMode=RCTM_User, time=0.9, value=0.15}
          {arriveTangent=-0.0405, interpMode=RCIM_Cubic, leaveTangent=-0.0405, tangentMode=RCTM_User, time=1, value=0}
         ]}}}
      CurveIndex=$Emitter.Age
     )
    )
    Scale Sprite Size Mode=::NewEnumerator2
   )
   &ScaleSpriteSizeBySpeed(
    Max Scale Factor=<0.3 2>
    Min Scale Factor=<0.3 0.3>
    Velocity Threshold:NiagaraFloat=&RandomRangeFloat(Minimum=50, Maximum=1000)
   ) {enabled=false}
   &Color(
    Color=&RandomRangeLinearColor(Minimum=<rgba 5 0 0 1>, Maximum=<rgba 3 1 0.1 1>)
    Scale Alpha=&FloatFromCurve(
     FloatCurve=DataInterface() {propertyValues {Curve {keys [{interpMode=RCIM_Cubic, tangentMode=RCTM_User, time=0, value=1}, {interpMode=RCIM_Cubic, tangentMode=RCTM_User, time=1, value=0}]}}}
     Scale Curve=3
    )
   )
   &ScaleColor(
    Scale Alpha=&FloatFromCurve(FloatCurve=DataInterface() {propertyValues {Curve {keys [
         {tangentMode=RCTM_User, time=0, value=1}
         {arriveTangent=-0.0405, interpMode=RCIM_Cubic, leaveTangent=-0.0405, tangentMode=RCTM_User, time=0.3, value=1}
         {arriveTangentWeight=1.3452, interpMode=RCIM_Cubic, tangentMode=RCTM_User, time=1, value=0}
        ]}}})
    Scale RGB=&RandomRangeVector(Minimum=<20 0 0>, Maximum=<1 10 1>)
    ScaleA=true
    ScaleRGB=true
   )
  ]
  rendererClasses [&NiagaraSpriteRendererProperties]
  renderers [{rendererClass=&NiagaraSpriteRendererProperties}]
 }
 Sparks_Floaters {
  emitterUpdateScript [
   &EmitterState(Delay First Loop Only=true, Life Cycle Mode=::NewEnumerator1, Loop Delay=0.5914, Loop Duration=1.9086, UseLoopDelay=true)
   &SpawnRate(SpawnRate=40, Use Spawn Probability=false)
  ]
  particleSpawnScript [
   &InitializeParticle(
    Color=&SelectLinearColorFromArray(
     Array Sampling Mode=::NewEnumerator0
     Direct Array Index=?
     Color Selection Array=DataInterface() {propertyValues {ColorData [<rgba 4 0.0133 0 1>, <rgba 0.401 0.0067 0 1>]}}
     Random Seed=0
    )
    Lifetime Max=1.5
    Lifetime Min=0.3
    Lifetime Mode=::NewEnumerator1
    Sprite Size Mode=::NewEnumerator1
    Uniform Sprite Size Max=1.5
    Uniform Sprite Size Min=0.5
    UsePositionOffset=false
    Write Material Random=true
   )
   &SkeletalMeshLocation(
    Bone / Socket Index=&RandomRangeInt(Minimum=0, Maximum=1)
    Bone Sampling Mode=::NewEnumerator4
    MaskSpawnGroup=false
    Sampled Position Offset=&FN_FrontEndVector3Modifier(
     Gameplay=&Lerp_Vector(A=<0 12 10>, B=<0 -8 19>, Alpha=$System.FN_Integrator001.Value)
     Front End=&Lerp_Vector(A=<0 11 -8>, B=<0 -7 19>, Alpha=$System.Dive)
    )
    Sampled Position Offset Coordinate Space:ENiagaraSkelMeshTransforms=::NewEnumerator2
    Skeletal Mesh=$User.SkeletalMesh
   )
   &ShapeLocation(
    MaskSpawnGroup=false
    Offset=&FN_FrontEndVector3Modifier(Gameplay=<0 0 0>, Front End=&Lerp_Vector(A=<0 0 30>, B=<0 -15 10>, Alpha=$System.FN_Integrator001.Value))
    Offset Mode=::NewEnumerator0
    Override Seed=false
    Sphere Radius=15
    UseOwnerScale=true
   )
   &InheritVelocity(Inherited Velocity Amount Scale=<0 0 -0.05>, Inherited Velocity Speed Limit=1000, Speed Limit=false, UseSpeedThreshold=false) {module=InheritVelocity001}
   &AddVelocity(
    Velocity=&Multiply_Vector(
     A=&Multiply_Vector(A=$System.SkydiveVelocity, B=<0 0 0>)
     B=&Lerp_Vector(
      A=&RandomRangeVector(Minimum=<0 0 0>, Maximum=<0 0 200>)
      B=&FN_FrontEndVector3Modifier(Gameplay=&RandomRangeVector(Minimum=<0 0 30>, Maximum=<0 0 30>), Front End=<0 0 60>)
      Alpha=$System.FN_Integrator001.Value
     )
    )
    Velocity Speed Scale=&FloatFromCurve(FloatCurve=DataInterface() {propertyValues {Curve {keys [{tangentMode=RCTM_User, time=0, value=1}, {tangentMode=RCTM_User, time=1, value=0}]}}})
   ) {enabled=false}
   &FN_SetSpawnGroup(Number of Spawn Groups=2)
   &DynamicMaterialParameters(
    Index 0 Param 1=HlslExpression() {hlslExpression="rand(1)"}
    Module.Param0WriteEnabled=true
    Module.Param1WriteEnabled=false
    Module.Param2WriteEnabled=false
    Module.Param3WriteEnabled=false
   ) {enabled=false}
   &ApplyInitialForces()
  ]
  particleUpdateScript [
   &InheritSourceMovement(ApplyForce=false, ApplyPosition=true, ApplyPositionFalloff=false) {enabled=false}
   &ParticleState()
   &FN_DivingAttributes_v2(Accumulate Dive=1, Accumulate Rest=0, ParaGlideLeanAlpha:NiagaraFloat=$User.ParaGlideLeanAlpha)
   &InheritVelocity(Inherited Velocity Amount Scale=<0.25 0.15 -0.02>, Speed Limit=true, UseSpeedThreshold=false) {enabled=false}
   &CurlNoiseForce(
    Noise Frequency=60
    Noise Quality / Cost=::NewEnumerator0
    Noise Strength=&FloatFromCurve(
     FloatCurve=DataInterface() {propertyValues {Curve {keys [
         {arriveTangent=4.1785, interpMode=RCIM_Cubic, leaveTangent=4.1785, tangentMode=RCTM_User, time=0, value=0.2}
         {arriveTangent=1.1179, interpMode=RCIM_Cubic, leaveTangent=1.1179, tangentMode=RCTM_User, time=0.2042, value=0.6228}
         {arriveTangent=0.0224, interpMode=RCIM_Cubic, leaveTangent=0.0224, tangentMode=RCTM_User, time=1, value=1}
        ]}}}
     Scale Curve=600
    )
    Pan Noise=true
    Pan Noise Field=<15 10 10>
   )
   &PointForce(
    Force Falloff Distance=200
    Force Strength=500
    Linear Force Vector:Vector3f=&RandomRangeVector(Minimum=<-1 -1 -0.25>, Maximum=<1 1 1>)
    Module.Use Falloff=true
    Module.UseLinearForce=true
   ) {enabled=false}
   &Drag(
    Drag=&FloatFromCurve(FloatCurve=DataInterface() {propertyValues {Curve {keys [
         {arriveTangent=4.1785, interpMode=RCIM_Cubic, leaveTangent=4.1785, tangentMode=RCTM_User, time=0, value=0}
         {arriveTangent=1.1179, interpMode=RCIM_Cubic, leaveTangent=1.1179, tangentMode=RCTM_User, time=0.2042, value=0.6228}
         {arriveTangent=0.0224, interpMode=RCIM_Cubic, leaveTangent=0.0224, tangentMode=RCTM_User, time=1, value=1}
        ]}}})
    Use Linear Drag=true
    Use Rotational Drag=true
   )
   &SolveForcesAndVelocity(Clamp Velocity=false, Limit Acceleration=false)
   &ScaleSpriteSize(
    Uniform Curve Scale=2
    Uniform Curve Sprite Scale=DataInterface() {propertyValues {Curve {keys [
        {arriveTangent=6.8904, interpMode=RCIM_Cubic, leaveTangent=6.8904, tangentMode=RCTM_User, time=0, value=0}
        {arriveTangent=10, interpMode=RCIM_Cubic, leaveTangent=10, time=0.0151, value=0.25}
        {arriveTangentWeight=0.0745, interpMode=RCIM_Cubic, leaveTangentWeight=0.2981, tangentMode=RCTM_User, time=0.1, value=1}
        {arriveTangent=-1.2657, interpMode=RCIM_Cubic, leaveTangent=-1.2657, time=0.2913, value=0.1}
        {interpMode=RCIM_Cubic, tangentMode=RCTM_User, time=0.337, value=0.7}
        {arriveTangent=-0.5124, interpMode=RCIM_Cubic, leaveTangent=-0.5124, tangentMode=RCTM_User, time=0.5849, value=0.15}
        {tangentMode=RCTM_User, time=1, value=0}
       ]}}}
   )
   &ScaleSpriteSizeBySpeed(Max Scale Factor=<0.6 3>, Min Scale Factor=<0.3 0.6>, Velocity Threshold=200) {enabled=false}
   &ScaleColor(
    Scale Alpha=&FloatFromCurve(
     FloatCurve=DataInterface() {propertyValues {Curve {keys [
         {arriveTangent=9.0725, interpMode=RCIM_Cubic, leaveTangent=9.0725, tangentMode=RCTM_User, time=0, value=0}
         {arriveTangent=-2.5249, interpMode=RCIM_Cubic, leaveTangent=-2.5249, tangentMode=RCTM_User, time=0.1124, value=0.9847}
         {arriveTangent=-0.1423, interpMode=RCIM_Cubic, leaveTangent=-0.1423, tangentMode=RCTM_User, time=0.2564, value=0.4594}
         {arriveTangent=2.0872, interpMode=RCIM_Cubic, leaveTangent=2.0872, time=0.4061, value=0.0866}
         {arriveTangent=-0.1886, interpMode=RCIM_Cubic, leaveTangent=-0.1886, time=0.4332, value=0.8284}
         {arriveTangent=-1.3751, interpMode=RCIM_Cubic, leaveTangent=-1.3751, time=0.6, value=0.05}
         {arriveTangent=-4.3712, interpMode=RCIM_Cubic, leaveTangent=-4.3712, tangentMode=RCTM_User, time=0.6352, value=0.5506}
         {arriveTangent=-0.0405, interpMode=RCIM_Cubic, leaveTangent=-0.0405, tangentMode=RCTM_User, time=1, value=0}
        ]}}}
     Scale Curve=60
    )
    ScaleA=true
    ScaleRGB=true
   )
  ]
  rendererClasses [&NiagaraSpriteRendererProperties]
  renderers [{rendererClass=&NiagaraSpriteRendererProperties}]
 }
}
summary {
 systemName=NS_Contrail_RippedHarvester
 emitters [
  {emitterName=DustFlow, rendererClasses [&NiagaraSpriteRendererProperties, &NiagaraMeshRendererProperties]}
  {emitterName=Smoke_Tail, rendererClasses [&NiagaraMeshRendererProperties]}
  {emitterName=Meteor_Mesh, rendererClasses [&NiagaraLightRendererProperties, &NiagaraMeshRendererProperties]}
  {emitterName=RockShards_Tail, rendererClasses [&NiagaraMeshRendererProperties]}
  {emitterName=RockShards_Floaters, rendererClasses [&NiagaraMeshRendererProperties]}
  {emitterName=GPU_Sparks, simTarget=GPUComputeSim, rendererClasses [&NiagaraSpriteRendererProperties]}
  {emitterName=Sparks_Floaters, rendererClasses [&NiagaraSpriteRendererProperties]}
  {emitterName=Flame_CenterPlume, rendererClasses [&NiagaraMeshRendererProperties]}
  {emitterName=Flame_Core, rendererClasses [&NiagaraMeshRendererProperties]}
  {emitterName=Flame_MeshLicks, rendererClasses [&NiagaraMeshRendererProperties]}
  {emitterName=Sparks_EnterBurst001, simTarget=GPUComputeSim, rendererClasses [&NiagaraSpriteRendererProperties]}
 ]
 userVariables [
  User.bIsFrontEnd {type=NiagaraBool}
  User.bIsFrontEndPreview {type=NiagaraBool, defaultValue=true}
  User.ColorA {type=LinearColor, description="", defaultValue {struct=LinearColor, value=<rgba 0.5559 0.6633 0.9583 1>}}
  User.ColorB {type=LinearColor, description="", defaultValue {struct=LinearColor, value=<rgba 0.7969 0.0266 0 1>}}
  User.IsSwing {type=NiagaraBool}
  User.ParaGlideLeanAlpha {type=NiagaraFloat, defaultValue=0, description=""}
  User.RibbonIDOffset {type=NiagaraInt32}
  User.SkeletalMesh {
   type=NiagaraDataInterfaceSkeletalMesh
   defaultValue {
    struct=NiagaraExt_VariableValue_DataInterface
    value {dataInterface=&self:NiagaraDataInterfaceSkeletalMesh_0}
   }
  }
  User.SkeletalMeshFrontend {
   type=NiagaraDataInterfaceSkeletalMesh
   defaultValue {
    struct=NiagaraExt_VariableValue_DataInterface
    value {dataInterface=&self:NiagaraDataInterfaceSkeletalMesh_1}
   }
  }
 ]
}
